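# Physics asset for the off-pawn version of the GainSneezeVital two-handed pickaxe, giving its handle and dangling net joints capsule collision bodies.
# A base net body joins the handle with free swing and twist, and three dynamic net bodies chain from it.
# Each chain link has asymmetric swing limits and very tight twist, so the net bends mostly along one axis.
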PhysicsAsset /BRCosmetics/Weapons/FORT_Melee/Pickaxe_GainSneezeVital/Meshes/GainSneezeVital_Axe_Physics_FE_OffPawn
bodies {
 Melee_TwoHanded_Handle [{
   shapeType=Capsule
   length=76.6912
   radius=6.8092
   shapeName=Melee_TwoHanded_Handle_capsule
   center=<-0.1191 -0.7974 23.062>
   rotation=<rot 3.0196 -90.0258 -0.2959>
  }]
 base_Net_jnt [{shapeType=Capsule, length=15, radius=1.1895, shapeName=base_Net_jnt_capsule}]
 dyn_Net_01_jnt [{shapeType=Capsule, length=8, radius=6, shapeName=dyn_Net_01_jnt_capsule, center=<2 0 0>, rotation=<rot -90 0 0>}]
 dyn_Net_02_jnt [{shapeType=Capsule, length=8, radius=6, shapeName=dyn_Net_02_jnt_capsule, center=<2 0 0>, rotation=<rot -90 0 0>}]
 dyn_Net_03_jnt [{shapeType=Capsule, length=8, radius=6, shapeName=dyn_Net_03_jnt_capsule, center=<2 0 0>, rotation=<rot -90 0 0>}]
}
constraints [
 {bone1Name=dyn_Net_01_jnt, bone2Name=base_Net_jnt, swing1LimitDegrees=40, swing2LimitDegrees=10, twistLimitDegrees=2.5}
 {bone1Name=dyn_Net_02_jnt, bone2Name=dyn_Net_01_jnt, swing1LimitDegrees=40, swing2LimitDegrees=10, twistLimitDegrees=2.5}
 {bone1Name=dyn_Net_03_jnt, bone2Name=dyn_Net_02_jnt, swing1LimitDegrees=40, swing2LimitDegrees=10, twistLimitDegrees=2.5}
 {
  bone1Name=base_Net_jnt
  bone2Name=Melee_TwoHanded_Handle
  swing1LimitDegrees=45
  swing1Motion=Free
  swing2LimitDegrees=45
  swing2Motion=Free
  twistLimitDegrees=45
  twistMotion=Free
 }
]
