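# Animation Blueprint for the GainSneezeVital pickaxe that adds rigid-body secondary motion to the axe skeleton.
# Three RigidBody nodes chain from the input pose, with default, front-end off-pawn and front-end on-pawn physics assets, simulated relative to b_AR_Root.
# Each update derives rigid-body-off and front-end pawn flags, then switches shared sim-space settings and gravity between front-end and in-game values.
# The on-pawn node reads its own dedicated sim-space settings.
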AnimBlueprint /BRCosmetics/Weapons/FORT_Melee/Pickaxe_GainSneezeVital/Meshes/GainSneezeVital_Axe_AnimBP
target_skeleton=&/GainSneezeVital_Axe_Skeleton
graphs [
 AnimGraph {nodes [
   AnimGraphNode_SubInput_0 = Input Pose()
   n1 = RigidBody!(
    ComponentPose=Local To Component(LocalPose=@AnimGraphNode_SubInput_0)
    Alpha
    OverrideWorldGravity:Vector=Get Override World Gravity() -> (Override World Gravity:Vector)
    ExternalForce
    SimSpaceSettings:Sim Space Settings Structure=Get Sim Space Settings() -> (Sim Space Settings:Sim Space Settings Structure)
   ) -> (Pose) {
    bOverrideWorldGravity=true
    overridePhysicsAsset=&/GainSneezeVital_Axe_Physics
    simulationSpace=BaseBoneSpace
    alphaBoolBlend {blendOption=Linear}
    baseBoneRef {boneName=b_AR_Root}
    componentAppliedLinearAccClamp=<3500 3500 3500>
    overrideWorldGravity {z=-980}
   }
   n2 = RigidBody!(
    ComponentPose=@n1
    bAlphaBoolEnabled:bool
    OverrideWorldGravity=Get Override World Gravity()
    ExternalForce
    SimSpaceSettings=Get Sim Space Settings()
   ) -> (Pose) {
    alphaInputType=Bool
    bOverrideWorldGravity=true
    overridePhysicsAsset=&/GainSneezeVital_Axe_Physics_FE_OffPawn
    simulationSpace=BaseBoneSpace
    alphaBoolBlend {blendOption=Linear}
    baseBoneRef {boneName=b_AR_Root}
    componentAppliedLinearAccClamp=<3500 3500 3500>
    overrideWorldGravity {z=-980}
   }
   n3 = RigidBody(
    ComponentPose=@n2
    OverrideWorldGravity=Get Override World Gravity()
    SimSpaceSettings=Get Sim Space Settings_FrontEnd_OnPawn() -> (Sim Space Settings_FrontEnd_OnPawn:Sim Space Settings Structure)
   ) {
    alphaInputType=Bool
    bOverrideWorldGravity=true
    overridePhysicsAsset=&/GainSneezeVital_Axe_Physics_FE_OnPawn
    simulationSpace=BaseBoneSpace
    alphaBoolBlend {blendOption=Linear}
    baseBoneRef {boneName=b_AR_Root}
    componentAppliedLinearAccClamp=<3500 3500 3500>
    overrideWorldGravity {z=-980}
   }
   Output Pose(Result=Component To Local(ComponentPose=@n3)) {layerGroup=None}
  ]}
 EventGraph {nodes [
   Event BlueprintUpdateAnimation()
   > Set RBAN off(RBAN off:bool=NOT Boolean(A=Get bRigidBodyEnabled())) -> (Output_Get:bool)
   > Set OnPlayerFE(OnPlayerFE:bool=AND Boolean(A=Get bIsInFrontEnd(), B=Get bIsOnPawn() -> (bIsOnPawn:bool))) -> (Output_Get:bool)
   n1 = Get bIsOnPawn()
   > Set OffPlayerFE(OffPlayerFE:bool=AND Boolean(A=Get bIsInFrontEnd(), B=NOT Boolean(A=@n1))) -> (Output_Get:bool)
   > n2 = Branch(Condition=Get bIsInFrontEnd())
   > Set Sim Space Settings(Sim Space Settings:Sim Space Settings Structure=Get Sim Space Settings_FrontEnd() -> (Sim Space Settings_FrontEnd:Sim Space Settings Structure)) -> (Output_Get:Sim Space Settings Structure)
   > Set Override World Gravity(Override World Gravity:Vector="0,0,-1660") -> (Output_Get:Vector)
   Set Sim Space Settings(execute=@n2.else, Sim Space Settings=Get Sim Space Settings_PIE() -> (Sim Space Settings_PIE:Sim Space Settings Structure))
   > Set Override World Gravity(Override World Gravity="0,0,-980")
  ]}
]
variables [
 OffPlayerFE {type=boolean}
 OnPlayerFE {type=boolean}
 Override World Gravity {type=Vector}
 RBAN off {type=boolean}
 RbanAlpha {type=number}
 RbanFEAlpha {type=number}
 Sim Space Settings {
  type=SimSpaceSettings
  default {
   dampingAlpha=1
   maxAngularAcceleration=10000
   maxAngularVelocity=10000
   maxLinearAcceleration=10000
   maxLinearVelocity=10000
   velocityScaleZ=1
   worldAlpha=0
   externalLinearDragV=<0 0 0>
  }
 }
 Sim Space Settings_FrontEnd {
  type=SimSpaceSettings
  default {
   dampingAlpha=0
   maxAngularAcceleration=750
   maxAngularVelocity=150
   maxLinearAcceleration=1500
   maxLinearVelocity=1000
   velocityScaleZ=0.25
   worldAlpha=1
   externalLinearDragV=<0.2 0.2 0.2>
  }
 }
 Sim Space Settings_FrontEnd_OnPawn {
  type=SimSpaceSettings
  default {
   dampingAlpha=0
   maxAngularAcceleration=50
   maxAngularVelocity=200
   maxLinearAcceleration=2250
   maxLinearVelocity=3000
   velocityScaleZ=0.375
   worldAlpha=1
   externalAngularVelocity=<0.5 0.5 0.5>
   externalLinearDragV=<0.5 0.5 0.5>
   externalLinearVelocity=<0.5 0.5 0.5>
  }
 }
 Sim Space Settings_PIE {
  type=SimSpaceSettings
  default {
   dampingAlpha=0
   maxAngularAcceleration=1000
   maxAngularVelocity=250
   maxLinearAcceleration=2000
   maxLinearVelocity=1500
   velocityScaleZ=0.5
   worldAlpha=1
   externalLinearDragV=<0.3 0.3 0.3>
  }
 }
]
